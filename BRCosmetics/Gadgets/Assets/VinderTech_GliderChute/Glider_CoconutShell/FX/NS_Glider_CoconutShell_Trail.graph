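# Niagara trail effect for the CoconutShell glider, combining speed-line sprites with wing ribbons.
# Speed_Lines spawns short-lived, velocity-stretched sprites on a torus, while Wing_Trail draws ten anim-trail ribbons between skeletal mesh socket pairs, tinted by time of day.
# A system timeline drives sprite spawn rate, ribbon velocity and opacity, with a system fade-in multiplying alpha.
# Front-end switches flip directions and adjust scale and intensity for preview.
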NiagaraSystem /BRCosmetics/Gadgets/Assets/VinderTech_GliderChute/Glider_CoconutShell/FX/NS_Glider_CoconutShell_Trail
dependencies {
 usedDataInterfaces [
  &NiagaraDataInterfaceVector2DCurve
  &NiagaraDataInterfaceCurve
  &NiagaraDataInterfaceColorCurve
  &NiagaraDataInterfaceArrayFloat
  &NiagaraDataInterfaceSkeletalMesh
 ]
 usedDynamicInputs [
  &SetBoolByFloatComparison
  &RemapRange
  &MakeIntFromFloat
  &NormalizeVector
  &TransformVector
  &FloatFromCurve
 ]
 usedModules [
  &self:SystemSpawnScript.SystemScriptSource.SystemScriptGraph.NiagaraNodeAssignment_0.SetVariables_#1
  &SystemState
  &Timeline
  &FN_Integrator
  &InterpolateOverTime
  &self:SystemSpawnScript.SystemScriptSource.SystemScriptGraph.NiagaraNodeAssignment_1.SetVariables_#2
 ]
}
emitters {
 Speed_Lines {
  emitterUpdateScript [
   &EmitterState()
   &SpawnRate(
    SpawnRate=&FN_FrontEndFloatModifier(Gameplay=&Lerp_Float(A=10, B=90, Alpha=$System.Timeline.NormalizedTime), Front End=50)
    Use Spawn Probability=true
   ) {module=SpawnRate001}
  ]
  particleSpawnScript [
   &InitializeParticle(
    Color=&MakeLinearColorFromVectorAndFloat(
     "Vector (RGB)"=<1 1 1>
     "Float (Alpha)"=&Multiply_Float(
      A=&UniformRangedFloat(Minimum=0.1, Maximum=0.18)
      B=&MakeCustomFloatFromBool(Bool=$User.bIsFrontEndPreview, True Float=0.5, False Float=1)
     )
    )
    Lifetime Max=0.16
    Lifetime Min=0.1
    Lifetime Mode=::NewEnumerator1
    Sprite Size Max=<3 10>
    Sprite Size Min=<2 2>
    Sprite Size Mode=::NewEnumerator2
    UsePositionOffset=false
    Write Material Random=true
   )
   &TorusLocation(
    Align Torus With Axis=true
    Handle Radius=0
    Large Radius:NiagaraFloat=&FN_FrontEndFloatModifier(
     Gameplay=&RandomRangeFloat(Minimum=200, Maximum=300)
     Front End=&RandomRangeFloat(Minimum=80, Maximum=120)
    )
    MaskSpawnGroup=false
    Module.Override Randomness=false
    Torus Axis=<0 1 0>
    Torus Origin:NiagaraPosition=&Add_Vector(
     A=Linked() {linkedVariable {name=Particles.Position, type=NiagaraPosition}}
     B=&MakeVector(X=0, Y=&FN_FrontEndFloatModifier(Gameplay=0, Front End=200), Z=0)
    )
   )
   &AddVelocityFromPoint(
    Module.Use Falloff Distance=false
    Origin Offset=<0 -500 0>
    Velocity Strength=&Multiply_Float(A=600, B=&FN_FrontEndFloatModifier(Gameplay=1, Front End=-1))
   )
   &AddVelocity(Velocity=&Multiply_Vector(
     A=&RandomRangeVector(Minimum=<0 300 0>, Maximum=<0 700 0>)
     B=&MakeVector(X=1, Y=&FN_FrontEndFloatModifier(Gameplay=1, Front End=-1), Z=1)
    ))
  ]
  particleUpdateScript [
   &SolveForcesAndVelocity(Clamp Velocity=false, Limit Acceleration=false)
   &ParticleState()
   &ScaleSpriteSize(Scale Factor=&Vector2DFromCurve(
     Vector2Curve=DataInterface() {propertyValues {XCurve {keys [<at 0 0.5 RCIM_Cubic>,<at 0.5 1 RCIM_Cubic>]}, YCurve {keys [<at 0 0 RCIM_Cubic>,<at 0.5 1 RCIM_Cubic>]}}}
     ScaleCurve=&Vector2DFromFloat(Value=1)
    ))
   &ScaleSpriteSizeBySpeed(
    Max Scale Factor=&FN_FrontEndVector2Modifier(Front End:Vector2f=<0.5 30>, Gameplay:Vector2f=<0.8 50>)
    Min Scale Factor=<0.2 1>
    Velocity Threshold=1
   )
   &ScaleColor(
    Scale Alpha=&Multiply_Float(
     A=&FN_FrontEndFloatModifier(Gameplay=&Lerp_Float(A=0, B=1, Alpha=$System.Timeline.NormalizedTime), Front End=0.5)
     B=$System.SystemFadeIn
    )
    Scale RGB=&VectorFromFloat(Value=1)
    ScaleA=true
    ScaleRGB=true
   )
  ]
  rendererClasses [&NiagaraSpriteRendererProperties]
  renderers [{rendererClass=&NiagaraSpriteRendererProperties}]
 }
 Wing_Trail {
  emitterSpawnScript [&FN_RibbonIDOffsetUtility()]
  emitterUpdateScript [
   &EmitterState(
    Inactive Response=::NewEnumerator1
    Life Cycle Mode=::NewEnumerator1
    Loop Behavior=::NewEnumerator1
    Loop Duration Mode=::NewEnumerator1
    UseLoopDelay=false
   )
   &FN_SpawnAnimTrail(Number of Anim Trails=10, Spawn Interval=0)
  ]
  particleSpawnScript [
   &FN_InitializeAnimTrailBoneIndecies(First Bone or Socket Index=1, Second Bone or Socket Index=0)
   &FN_InitializeAnimTrailBoneIndecies(AnimTrail Selection=1, First Bone or Socket Index=3, Second Bone or Socket Index=2) {module=FN_InitializeAnimTrailBoneIndecies001}
   &FN_InitializeAnimTrailBoneIndecies(AnimTrail Selection=2, First Bone or Socket Index=5, Second Bone or Socket Index=4) {module=FN_InitializeAnimTrailBoneIndecies002}
   &FN_InitializeAnimTrailBoneIndecies(AnimTrail Selection=3, First Bone or Socket Index=7, Second Bone or Socket Index=6) {module=FN_InitializeAnimTrailBoneIndecies003}
   &FN_InitializeAnimTrailBoneIndecies(AnimTrail Selection=4, First Bone or Socket Index=9, Second Bone or Socket Index=8) {module=FN_InitializeAnimTrailBoneIndecies004}
   &FN_InitializeAnimTrailBoneIndecies(AnimTrail Selection=5, First Bone or Socket Index=11, Second Bone or Socket Index=10) {module=FN_InitializeAnimTrailBoneIndecies005}
   &FN_InitializeAnimTrailBoneIndecies(AnimTrail Selection=6, First Bone or Socket Index=13, Second Bone or Socket Index=12) {module=FN_InitializeAnimTrailBoneIndecies006}
   &FN_InitializeAnimTrailBoneIndecies(AnimTrail Selection=7, First Bone or Socket Index=15, Second Bone or Socket Index=14) {module=FN_InitializeAnimTrailBoneIndecies007}
   &FN_InitializeAnimTrailBoneIndecies(AnimTrail Selection=8, First Bone or Socket Index=17, Second Bone or Socket Index=16) {module=FN_InitializeAnimTrailBoneIndecies008}
   &FN_InitializeAnimTrailBoneIndecies(AnimTrail Selection=9, First Bone or Socket Index=19, Second Bone or Socket Index=18) {module=FN_InitializeAnimTrailBoneIndecies009}
   &FN_InitializeAnimTrail(
    Color=&FNTimeOfDayColor(
     Night Color=<rgba 0.2 0.3467 1 2>
     Morning Color=<rgba 0.36 0.612 0.9 2>
     Day Color=<rgba 0.3604 0.5947 0.901 2>
     Evening Color=<rgba 0.2703 0.5436 0.901 2>
    )
    Lifetime=0.175
    Position Offset=&FN_FrontEndVector3Modifier(Gameplay=&Lerp_Vector(A=<0 -6 0>, B=<0 -9 0>, Alpha=$System.Timeline.Time), Front End=<0 5 0>)
    SkeletalMesh=$System.Skeletal Mesh
   )
   &AddVelocity(Velocity=&Multiply_VectorByFloat(
     Vector=&Lerp_Vector(A=<0 400 0>, B=<0 800 0>, Alpha=$System.Timeline.Time)
     Float=&MakeCustomFloatFromBool(Bool=$User.bIsFrontEnd, True Float=-1, False Float=1)
    ))
   &DynamicMaterialParameters(
    Index 0 Param 1=1
    Index 0 Param 2=&SelectFloatFromArray(
     Direct Array Index=&SpawnGroup()
     Float Selection Array=DataInterface() {propertyValues {FloatData [0,0.12,0.28,0.36,0.49,0.53,0.66,0.73,0.81,0.96]}}
    )
    Module.Index0Param0Write=true
    Module.Index0Param1Write=true
    Module.Index0Param2Write=true
    Module.Index0Param3Write=true
   ) {enabled=false}
  ]
  particleUpdateScript [
   &SolveForcesAndVelocity(Clamp Velocity=false, Limit Acceleration=false)
   &FN_UpdateAnimTrail(Locked Particle RibbonWidth=1, SkeletalMesh=$System.Skeletal Mesh) {module=FN_UpdateAnimTrail001}
   &ScaleColor(
    Scale Alpha=&Multiply_Float(A=&Lerp_Float(A=0.25, B=0.5, Alpha=$System.Timeline.NormalizedTime), B=$System.SystemFadeIn)
    ScaleA=true
    ScaleRGB=true
   )
  ]
  rendererClasses [&NiagaraRibbonRendererProperties]
  renderers [{rendererClass=&NiagaraRibbonRendererProperties}]
 }
}
summary {
 systemName=NS_Glider_CoconutShell_Trail
 emitters [
  {emitterName=Speed_Lines, rendererClasses [&NiagaraSpriteRendererProperties]}
  {emitterName=Wing_Trail, rendererClasses [&NiagaraRibbonRendererProperties]}
 ]
 userVariables [
  User.bIsFrontEnd {type=NiagaraBool}
  User.bIsFrontEndPreview {type=NiagaraBool}
  User.bIsFullyDeployed {type=NiagaraBool}
  User.ForwardDot {type=NiagaraFloat, defaultValue=0, description=""}
  User.RibbonIDOffset {type=NiagaraInt32}
  User.RightDot {type=NiagaraFloat, defaultValue=0, description=""}
 ]
}
