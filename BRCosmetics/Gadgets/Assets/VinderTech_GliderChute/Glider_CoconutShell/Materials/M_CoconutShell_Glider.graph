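# Masked material for the Coconut Shell glider, adding time-of-day night glow, Fresnel flight emissive and leaf-like translucency between character parent Pre/PostFX functions.
# Tech-art mask channels drive glow, dithered opacity and a per-pixel blend from the parent shading into Subsurface.
# Night values adjust base color and specular, with front-end parameters for overrides.
# Transmission color comes from a texture weighted by normals and a sun-direction light mask with fallback.
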Material /BRCosmetics/Gadgets/Assets/VinderTech_GliderChute/Glider_CoconutShell/Materials/M_CoconutShell_Glider
graph {
 properties {
  bUseMaterialAttributes=true
  bUsedWithMeshDeformer=true
  bUsedWithMorphTargets=true
  bUsedWithSkeletalMesh=true
  blendMode=BLEND_Masked
  pixelDepthOffsetMode=PDOM_Legacy
  shadingModel=MSM_FromMaterialExpression
  usedShadingModels=Default Lit | Subsurface | Clear Coat | Subsurface Profile | Cloth
 }
 nodes [
  AtmosphericLightVector()
  Custom() {code="return View.DirectionalLightDirection;", description=DirectionalLightVector, inputs []}
  n1 = MaterialFunctionCall() {materialFunction=&MF_CharacterParent_2023_PreFX}
  n2 = GetMaterialAttributes(None=@n1.Result) {attributeGetTypes [
    #1
    #2
    #3
    #4
    #5
   ]}
  n3 = LinearInterpolate(A=@n2.Opacity Mask, Alpha=NamedRerouteUsage()) {constB=0}
  n4 = MaterialFunctionCall!(Alpha Threshold=Constant() {r=0.2}, Random) {materialFunction=/Game/Environments/Asteria/Sets/NeonCity/Techart/Screen/MaterialFunctions/DitherTemporalAABlue}
  n5 = VectorParameter() {group=Reactivity_Night, parameterName=NightEM, channelNames {a=Intensity, b="", g="", r=""}, defaultValue=<rgba 0 0.3066 1 2.5>}
  n6 = MaterialFunctionCall(Value_Morning=Constant() {r=0.5}, Value_Day=Constant(), Value_Evening=Constant() {r=0.5}, Value_Night=Constant() {r=1}) {materialFunction=&Scalar_Time_of_Day}
  n7 = Multiply(A=ScalarParameter() {group=FrontEnd, parameterName=FrontEndAlpha}) {constB=3}
  n8 = NamedRerouteUsage()
  n9 = Multiply(A=NamedRerouteUsage(), B=Multiply(A=Multiply(A=@n5, B=@n5), B=LinearInterpolate(A=@n6.Result, B=@n7, Alpha=@n8)))
  n10 = NamedRerouteUsage()
  n11 = ScalarParameter() {defaultValue=0.1, group=Reactivity_Night, parameterName=NightSpec_Intensity}
  n12 = SetMaterialAttributes!(
   MaterialAttributes=@n2.MaterialAttributes
   Opacity Mask=Saturate(None=Add(A=@n3, B=@n4.Result))
   Emissive Color=Add(A=@n2.Emissive Color, B=Add(A=@n9, B=@n10))
   Base Color=NamedRerouteUsage()
   Specular=LinearInterpolate(A=@n2.Specular, B=Multiply(A=@n2.Specular, B=@n11) {constB=0.8}, Alpha=NamedRerouteUsage())
  ) {attributeSetTypes [
    #1
    #2
    #3
    #4
   ]}
  n13 = ShadingModel() {shadingModel=MSM_Subsurface}
  n14 = NamedRerouteUsage()
  n15 = ComponentMask(None=NamedRerouteUsage()) {b=true}
  MaterialFunctionCall(
   Material=BlendMaterialAttributes(
    A=@n12
    B=SetMaterialAttributes!(MaterialAttributes=@n12, Shading Model=@n13, Fuzz Color or Subsurface Color=@n14, Opacity=@n15) {attributeSetTypes [#6, #7, #8]}
    Alpha=NamedRerouteUsage()
   )
   M=@n1.M
   Pre Skinned Local Position=@n1.M
   Distance Blend=@n1.M
   Fresnel=@n1.M
  ) {materialFunction=&MF_CharacterParent_2023_PostFX}
  NamedRerouteDeclaration(None=@n6.Result) {name=NightAlpha01, nodeColor=<rgba 0.9219 1 0 1>}
  n16 = Constant()
  NamedRerouteDeclaration(None=MaterialFunctionCall(Value_Morning=@n16, Value_Day=@n16, Value_Evening=Constant(), Value_Night=Constant() {r=1}) {materialFunction=&Scalar_Time_of_Day}.Result) {name=NightAlpha02, nodeColor=<rgba 0.9219 1 0 1>}
  n17 = VectorParameter() {group=Reactivity_Night, parameterName=FlyEM, channelNames {a=Intensity, b="", g="", r=""}, defaultValue=<rgba 0.3229 0.7092 1 2>}
  n18 = Multiply(A=LinearInterpolate(Alpha=Fresnel() {baseReflectFraction=0, exponent=2}) {constA=0.2}, B=@n17)
  n19 = NamedRerouteUsage()
  n20 = Constant() {r=1}
  n21 = Constant() {r=0.8}
  n22 = Constant() {r=0.6}
  NamedRerouteDeclaration(None=Multiply(
    A=Multiply(A=@n18, B=@n19)
    B=MaterialFunctionCall(Value_Morning=@n20, Value_Day=@n20, Value_Evening=@n21, Value_Night=@n22) {materialFunction=&Scalar_Time_of_Day}.Result
   )) {name=FlyEM, nodeColor=<rgba 0.1953 0 1 1>}
  NamedRerouteDeclaration(None=@n1.M) {name=M_Map, nodeColor=<rgba 0 0.8359 1 1>}
  NamedRerouteDeclaration(None=ScalarParameter() {group=FrontEnd, parameterName=IsInFrontEnd}) {name=FrontEndOverride, nodeColor=<rgba 1 0.3047 0 1>}
  n23 = NamedRerouteUsage()
  n24 = ChannelMaskParameter(None=@n23) {group=Reactivity_Night, maskChannel=Blue, parameterName=BioMask_TechArt, defaultValue=<rgba 0 0 1 0>}
  NamedRerouteDeclaration(None=@n24.R) {name=BioMask, nodeColor=<rgba 0 0.2031 1 1>}
  NamedRerouteDeclaration(None=ChannelMaskParameter(None=@n23) {group=Glider_TechArt, parameterName=OpacityMask_TechArt}.R) {name=OpacityMask, nodeColor=<rgba 0.5677 0 0.0526 1>}
  NamedRerouteDeclaration(None=@n2.Base Color) {name=BaseColor_IN, nodeColor=<rgba 1 0 0.5625 1>}
  n25 = NamedRerouteUsage()
  n26 = LinearInterpolate(A=LinearInterpolate(A=NamedRerouteUsage(), B=@n25), B=@n25, Alpha=NamedRerouteUsage())
  n27 = Multiply(A=@n26, B=ScalarParameter() {defaultValue=0.6, group=Reactivity_Night, parameterName=NightBaseColor_Intensity})
  n28 = NamedRerouteUsage()
  n29 = NamedRerouteUsage()
  n30 = NamedRerouteUsage()
  NamedRerouteDeclaration(None=LinearInterpolate(A=LinearInterpolate(A=@n26, B=@n27, Alpha=@n28), Alpha=Multiply(A=@n29, B=@n30)) {constB=0}) {name=BaseColor_OUT, nodeColor=<rgba 0 0.4844 1 1>}
  NamedRerouteDeclaration(None=@n1.Tech Art Mask) {name=TechArtMask_IN, nodeColor=<rgba 0.6172 1 0 1>}
  n31 = Custom() {code="return View.DirectionalLightDirection;", description=DirectionalLightVector, inputs []}
  n32 = DotProduct(A=@n31, B=@n31)
  n33 = CollectionParameter() {
   collection=&FortniteMaterialParameters
   parameterName=SunlightDirectionalVector
  }
  n34 = ComponentMask(None=@n33) {b=true, g=true, r=true}
  n35 = MaterialFunctionCall!(In=If(
    A=Custom() {code="return View.DirectionalLightColor;", description=DirectionalLightColor, inputs []}
    "A > B"=If(A=@n32, "A > B"=@n31, "A == B"=@n34, "A < B"=@n34) {desc=check if DirectionalLight exists}
    "A == B"=@n34
    "A < B"=@n34
   ) {desc=check if DirectionalLight color is Black}) {materialFunction=/Game/Packages/Fortress_SharedMaterials/Cinematics/Anime/MF_CINE_AnimeLightDirOverride}
  n36 = MaterialFunctionCall(Float3=@n35.Result) {materialFunction=&BreakOutFloat3Components}
  n37 = ScalarParameter() {
   desc="0.0 = the light vector is locked on the XY plane.\r\n1.0 = the actual light vector is used  \r\n(for Faces use a blend value of 0,0)"
   group=02_Anime Parameters
   parameterName=LockedZLightVector_Blend
   sliderMax=1
  }
  n38 = MaterialFunctionCall(X=@n36.R, Y=@n36.R, Z=Multiply(A=@n36.B, B=@n37)) {materialFunction=&MakeFloat3}
  n39 = StaticSwitchParameter(True=LinearInterpolate(A=@n35.Result, B=Normalize(VectorInput=@n38.Result), Alpha=VertexColor().A), False=@n35.Result) {
   desc="Lock the Z axis of the Light Vector\r\nis driven by the VertexColor A channel (0 = Regular LightVector, 1= Locked LIghtVector)"
   group=02_Anime Parameters
   parameterName=UseLockedZLightVector
   sortPriority=0
  }
  NamedRerouteDeclaration(None=@n39) {name=LightVector, nodeColor=<rgba 1 0.934 0.5103 1>}
  n40 = NamedRerouteUsage()
  n41 = PixelNormalWS()
  NamedRerouteDeclaration(None=Saturate(None=DotProduct(A=@n40, B=@n41))) {name=LightMask, nodeColor=<rgba 0 0.9766 1 1>}
  n42 = ChannelMaskParameter(None=@n23) {group=Glider_TechArt, maskChannel=Green, parameterName=SSSMask_TechArt, defaultValue=<rgba 0 1 0 0>}
  NamedRerouteDeclaration(None=@n42.R) {name=SSSMask, nodeColor=<rgba 0.0994 0.4323 0.0996 1>}
  NamedRerouteDeclaration(None=@n2.Normal) {name=Normal_IN, nodeColor=<rgba 1 0.0234 0 1>}
  n43 = OneMinus(None=Distance(A=NamedRerouteUsage(), B=Constant3Vector() {constant=<rgba 0 0 1 0>}))
  n44 = TextureSampleParameter2D() {
   group=Glider_TechArt
   parameterName=TransmitColor
   texture=/BRCosmetics/Gadgets/Assets/VinderTech_GliderChute/Glider_CoconutShell/Textures/T_CoconutShell_Glider_TransmissColor
  }
  n45 = NamedRerouteUsage()
  n46 = Multiply(A=@n44.RGB, B=SmoothStep(Value=OneMinus(None=@n45)) {constMin=0.2})
  NamedRerouteDeclaration(None=LinearInterpolate(A=Multiply(A=@n43, B=@n46), Alpha=NamedRerouteUsage()) {constB=0}) {name=TransmitColor, nodeColor=<rgba 1 0.375 0 1>}
  NamedRerouteUsage()
  NamedRerouteUsage()
  ScalarParameter() {defaultValue=0.9, parameterName=SSS_Opacity}
  SkyAtmosphereLightDirection()
 ]
}
